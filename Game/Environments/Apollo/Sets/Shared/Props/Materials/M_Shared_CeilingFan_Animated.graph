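# Nanite-enabled Default Lit material for an animated ceiling fan prop shared across Apollo environment sets.
# Diffuse, Normal and Specular texture parameters, defaulting to Tinfoil light fixture textures, build the material attributes.
# The Specular mask's green channel feeds Metallic, Specular and Roughness alike.
# The attributes pass through MF_FanSpin, borrowed from the Creative MilitaryBase props, which by its name supplies the spin animation.
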Material /Game/Environments/Apollo/Sets/Shared/Props/Materials/M_Shared_CeilingFan_Animated
graph {
 properties {bUseMaterialAttributes=true, bUsedWithNanite=true, pixelDepthOffsetMode=PDOM_Legacy, usedShadingModels=Default Lit}
 nodes [
  n1 = TextureSampleParameter2D() {parameterName=Specular, samplerType=SAMPLERTYPE_Masks, texture=/Game/Environments/Apollo/Sets/Tinfoil/Textures/T_Tinfoil_LightFixture_01_S}
  n2 = TextureSampleParameter2D() {parameterName=Normal, samplerType=SAMPLERTYPE_Normal, texture=/Game/Environments/Apollo/Sets/Tinfoil/Textures/T_Tinfoil_LightFixture_01_N}
  MaterialFunctionCall!(Material Attributes=MakeMaterialAttributes!(
    BaseColor=TextureSampleParameter2D() {parameterName=Diffuse, texture=/Game/Environments/Apollo/Sets/Tinfoil/Textures/T_Tinfoil_LightFixture_01_D}.RGB
    Metallic=@n1.G
    Specular=@n1.G
    Roughness=@n1.G
    Anisotropy
    EmissiveColor
    Opacity
    OpacityMask
    Normal=@n2.RGB
    Tangent
    WorldPositionOffset
    SubsurfaceColor
    ClearCoat
    ClearCoatRoughness
    AmbientOcclusion
    Refraction
    CustomizedUVs_0
    CustomizedUVs_1
    CustomizedUVs_2
    CustomizedUVs_3
    CustomizedUVs_4
    CustomizedUVs_5
    CustomizedUVs_6
    CustomizedUVs_7
    PixelDepthOffset
    ShadingModel
    Displacement
   )) {materialFunction=/Game/Creative/Sets/MilitaryBase/Props/MF_FanSpin}
 ]
}
