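# Material function that applies an animated flicker to an input value, for shared Apollo prop materials.
# Time scaled by FlickerRate is wrapped with Frac to sample a looping curve from a light-flicker curve atlas.
# A channel mask parameter selects one curve channel, which is saturated and multiplied into the input.
# A scalar parameter, off by default, blends between the original and flickered value.
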MaterialFunction /Game/Environments/Apollo/Sets/Shared/Props/Materials/MF_LightFlickering
graph {nodes [
  n1 = FunctionInput()
  n2 = Time()
  n3 = ScalarParameter() {defaultValue=0.5, parameterName=FlickerRate}
  n4 = CurveAtlasRowParameter(CurveTime=Frac(None=Multiply(A=@n2, B=@n3))) {
   atlas=/Game/Environments/Apollo/Sets/Shared/Props/Textures/CA_Shared_LightFlickers
   curve=/Game/Environments/Apollo/Sets/Shared/Props/Textures/CC_Shared_FlickerCurve_03
   defaultValue=2
   parameterName=Curve
  }
  n5 = Saturate(None=ChannelMaskParameter(None=@n4) {parameterName=Flicker Channel}.R)
  FunctionOutput(None=LinearInterpolate(A=@n1, B=Multiply(A=@n1, B=@n5), Alpha=ScalarParameter() {parameterName=x13}))
 ]}
