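# M_Shared_EmissiveElement is a Nanite-enabled Default Lit material for flickering emissive elements on shared Apollo props.
# A color parameter, optionally multiplied by an emissive map, is scaled by boost and divided-down intensity parameters, then passed through MF_LightFlickering.
# The Spec/Roughness alpha masks the result and, via static switch, can replace the tint texture as emissive map.
# Diffuse and normal textures are also exposed as parameters.
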Material /Game/Environments/Apollo/Sets/Shared/Props/Materials/M_Shared_EmissiveElement
graph {
 properties {bUsedWithNanite=true, pixelDepthOffsetMode=PDOM_Legacy, usedShadingModels=Default Lit}
 nodes [
  LinearInterpolate()
  n1 = TextureSampleParameter2D() {parameterName=Spec/Roughness, samplerType=SAMPLERTYPE_Masks, texture=&T_BlackMask}
  n2 = StaticSwitchParameter(
   True=@n1.A
   False=TextureSampleParameter2D() {parameterName=EmissiveTintMap, texture=/Game/Environments/Apollo/Sets/Shared/Props/Textures/T_Shared_Signs_D}.RGB
  ) {parameterName=Use Spec Alpha for Emissive Mask?}
  n3 = VectorParameter() {parameterName=x12, defaultValue=<rgba 0.6198 0.6198 0.6198 1>}
  n4 = Multiply(
   A=StaticSwitch(True=Multiply(A=@n2, B=@n3.RGB), False=@n3.RGB, Value=StaticBoolParameter() {parameterName=UseEmissiveMap})
   B=ScalarParameter() {defaultValue=2, parameterName=Emissive Boost}
  )
  n5 = Divide(A=ScalarParameter() {defaultValue=1, parameterName=x11}) {constB=10}
  Multiply(
   A=@n1.A
   B=MaterialFunctionCall!(In=Multiply(A=@n4, B=@n5)) {materialFunction=&/MF_LightFlickering}.Result
  )
  TextureSampleParameter2D() {parameterName=Normal, samplerType=SAMPLERTYPE_Normal, texture=/Engine/EngineMaterials/FlatNormal}
  TextureSampleParameter2D() {parameterName=Diffuse, texture=&T_BlacksRGB}
 ]
}
